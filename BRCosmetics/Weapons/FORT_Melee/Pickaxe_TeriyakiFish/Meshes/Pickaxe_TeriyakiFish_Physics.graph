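# Physics asset for the TeriyakiFish two-handed pickaxe, giving capsule bodies and joints to the props attached to its staff.
# The staff base joins the handle and roots three chains: a two-part clay pot, a three-part sack, and ribbons ending in a two-part boot.
# The first ribbon is angularly locked to the staff base, while the lower pot and boot joints each lock one swing axis.
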PhysicsAsset /BRCosmetics/Weapons/FORT_Melee/Pickaxe_TeriyakiFish/Meshes/Pickaxe_TeriyakiFish_Physics
bodies {
 Melee_TwoHanded_Handle [{shapeType=Capsule, length=86.3687, radius=1.3198, center=<0.7612 -1.1807 57.4451>, rotation=<rot 0.7591 -90 0.136>}]
 dyn_boot_01 [{shapeType=Capsule, length=5, radius=6, center=<6 0 0>, rotation=<rot 0 -90 0>}]
 dyn_boot_02 [
  {shapeType=Capsule, length=8, radius=7.5, center=<3 0 0>, rotation=<rot 0 90 -90>}
  {shapeType=Capsule, length=20.6428, radius=6.3572, center=<15 7 0>, rotation=<rot 0 -5 85>}
 ]
 dyn_clayPot_01 [{shapeType=Capsule, length=10.9485, radius=3.6683, center=<5.8345 0.5189 1.0175>, rotation=<rot 3.3086 -89.9992 -84.5036>}]
 dyn_clayPot_02 [
  {shapeType=Capsule, length=0.75, radius=12.3504, center=<11.7267 -0.9626 0.6292>, rotation=<rot 8.0706 -89.9994 -82.227>}
  {shapeType=Capsule, length=7.75, radius=9.5, center=<3.6992 0.0005 0>, rotation=<rot -90 -0.0396 0.0397>}
 ]
 dyn_ribbons_01 [{shapeType=Capsule, length=24.2788, radius=2.4964, center=<-0.1344 -0.9692 -0.8907>, rotation=<rot 0 -178.4899 -58.1487>}]
 dyn_ribbons_02 [{shapeType=Capsule, length=22, radius=1.2, center=<9.8176 0 0>, rotation=<rot 0 90 -90>}]
 dyn_sack_01 [{shapeType=Capsule, length=0.505, radius=4.255, rotation=<rot 0 -90 0>}]
 dyn_sack_02 [{shapeType=Capsule, length=8.0591, radius=2.6338, center=<1.3881 0.0765 0.2345>, rotation=<rot 0 -97.2119 -85.883>}]
 dyn_sack_03 [{shapeType=Capsule, length=0.3531, radius=5.8408, center=<1.1317 -0.1491 0.1075>, rotation=<rot 79.6856 102.3385 -167.8544>}]
 dyn_staff_base [
  {shapeType=Capsule, length=24.761, radius=2, center=<0.6209 8.2104 12.5156>, rotation=<rot 0.0001 171.831 -33.4263>}
  {shapeType=Capsule, length=15.511, radius=2.5, center=<-1.2204 -5.0958 7.2266>, rotation=<rot 7.6442 -6.4664 -40.4329>}
 ]
}
constraints [
 {bone1Name=dyn_staff_base, bone2Name=Melee_TwoHanded_Handle, swing1LimitDegrees=45, swing2LimitDegrees=45, twistLimitDegrees=45}
 {bone1Name=dyn_clayPot_02, bone2Name=dyn_clayPot_01, swing1LimitDegrees=60, swing2LimitDegrees=45, swing2Motion=Locked, twistLimitDegrees=5}
 {bone1Name=dyn_boot_02, bone2Name=dyn_boot_01, swing1LimitDegrees=45, swing2LimitDegrees=45, swing2Motion=Locked, twistLimitDegrees=5}
 {bone1Name=dyn_clayPot_01, bone2Name=dyn_staff_base, swing1LimitDegrees=60, swing2LimitDegrees=60, twistLimitDegrees=5}
 {bone1Name=dyn_boot_01, bone2Name=dyn_ribbons_02, swing1LimitDegrees=45, swing2LimitDegrees=45, twistLimitDegrees=5}
 {bone1Name=dyn_ribbons_02, bone2Name=dyn_ribbons_01, swing1LimitDegrees=45, swing2LimitDegrees=90, twistLimitDegrees=5}
 {
  bone1Name=dyn_ribbons_01
  bone2Name=dyn_staff_base
  swing1LimitDegrees=45
  swing1Motion=Locked
  swing2LimitDegrees=45
  swing2Motion=Locked
  twistLimitDegrees=45
  twistMotion=Locked
 }
 {bone1Name=dyn_sack_01, bone2Name=dyn_staff_base, swing1LimitDegrees=30, swing2LimitDegrees=30, twistLimitDegrees=5}
 {bone1Name=dyn_sack_02, bone2Name=dyn_sack_01, swing1LimitDegrees=10, swing2LimitDegrees=45, twistLimitDegrees=5}
 {bone1Name=dyn_sack_03, bone2Name=dyn_sack_02, swing1LimitDegrees=45, swing2LimitDegrees=45, twistLimitDegrees=5}
]
