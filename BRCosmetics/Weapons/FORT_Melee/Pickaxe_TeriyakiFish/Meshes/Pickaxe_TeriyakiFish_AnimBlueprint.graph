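# Animation Blueprint for the TeriyakiFish pickaxe that adds procedural secondary motion to its skeleton.
# The input pose runs through a RigidBody node using the pickaxe's own physics asset.
# An Anim Dynamics node then simulates the dyn_water bone as a sprung, heavily damped body with limited angular swing, suggesting a sloshing water element.
# The event graph holds only unconnected default nodes and there are no variables.
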AnimBlueprint /BRCosmetics/Weapons/FORT_Melee/Pickaxe_TeriyakiFish/Meshes/Pickaxe_TeriyakiFish_AnimBlueprint
target_skeleton=&/Pickaxe_TeriyakiFish_Skeleton
graphs [
 AnimGraph {nodes [
   AnimGraphNode_SubInput_0 = Input Pose()
   n1 = RigidBody(ComponentPose=Local To Component(LocalPose=@AnimGraphNode_SubInput_0)) {
    overridePhysicsAsset=&/Pickaxe_TeriyakiFish_Physics
    alphaBoolBlend {blendOption=Linear}
    componentLinearAccScale=<1 1 1>
    componentLinearVelScale=<1 1 1>
   }
   n2 = Anim Dynamics - Bone: dyn_water(ComponentPose=@n1, Alpha=1, ExternalForce="0,0,0") {
    angularDampingOverride=0.999
    angularSpringConstant=1000
    bAngularSpring=true
    bGravityOverrideInSimSpace=true
    bOverrideAngularDamping=true
    alphaBoolBlend {blendOption=Linear}
    boundBone {boneName=dyn_water}
    componentAppliedLinearAccClamp=<3000 3000 3000>
    componentLinearAccScale=<1 1 1>
    componentLinearVelScale=<1 1 1>
    physicsBodyDefinitions [{
      sphereCollisionRadius=0
      boundBone {boneName=dyn_water}
      boxExtents=<14 30 3>
      constraintSetup {angularTargetAxis=AxisY, twistAxis=AxisY, angularLimitsMax=<45 45 0>, angularLimitsMin=<-45 -45 0>, angularTarget=<0 0 0>}
      localJointOffset=<0 -0.05 -7>
     }]
   }
   Output Pose(Result=Component To Local(ComponentPose=@n2)) {layerGroup=None}
  ]}
 EventGraph {nodes [
   TryGetPawnOwner()
   Event BlueprintUpdateAnimation()
  ]}
]
variables []
